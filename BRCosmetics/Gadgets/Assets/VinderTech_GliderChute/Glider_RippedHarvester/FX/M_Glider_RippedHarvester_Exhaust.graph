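# Translucent, unlit, two-sided material for the Ripped Harvester glider's exhaust flames on Niagara mesh particles.
# Perlin noise distorts panning lava-noise UVs, with a FlamesTurnDirection dynamic parameter skewing the flow.
# A vertical gradient and double-sided Fresnel shape the mask, which blends two colours from palettes selected by the Variant parameter.
# Particle colour tints emissive and scales opacity, which gets depth fade except on mobile.
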Material /BRCosmetics/Gadgets/Assets/VinderTech_GliderChute/Glider_RippedHarvester/FX/M_Glider_RippedHarvester_Exhaust
graph {
 properties {
  bUsedWithNiagaraMeshParticles=true
  blendMode=BLEND_Translucent
  pixelDepthOffsetMode=PDOM_Legacy
  shadingModel=MSM_Unlit
  twoSided=true
  usedShadingModels=Unlit
 }
 nodes [
  n1 = DynamicParameter() {defaultValue=<rgba 0 1 1 1>, paramNames [Variant, FlamesTurnDirection, Param3, Param4]}
  n2 = LinearInterpolate(
   A=VectorParameter() {parameterName=Color1, defaultValue=<rgba 0.5 0.0083 0 1>}.RGB
   B=VectorParameter() {parameterName=ColorAlt1, defaultValue=<rgba 0 0.0933 0.7 1>}.RGB
   Alpha=@n1.Variant
  )
  n3 = LinearInterpolate(
   A=VectorParameter() {parameterName=Color2, defaultValue=<rgba 5 1.1133 0.6 1>}.RGB
   B=VectorParameter() {parameterName=ColorAlt2, defaultValue=<rgba 0.5 2 5 1>}.RGB
   Alpha=@n1.Variant
  )
  n4 = TextureCoordinate()
  n5 = MaterialFunctionCall(Float2=@n4) {materialFunction=&BreakOutFloat2Components}
  n6 = Multiply(A=@n5.R)
  n7 = Multiply(A=@n5.G) {constB=0.3}
  n8 = TextureSampleParameter2D(UVs=Panner(Coordinate=AppendVector(A=@n6, B=@n7)) {speedY=-0.15}) {
   parameterName=TexDistort
   samplerType=SAMPLERTYPE_LinearGrayscale
   texture=&T_SimplePerlin_Greyscale
  }
  n9 = MaterialFunctionCall(Float2=Add(A=Multiply(A=@n8.R) {constB=0.1}, B=@n4) {constB=0}) {materialFunction=&BreakOutFloat2Components}
  n10 = OneMinus(None=@n9.G)
  n11 = Saturate(None=Power(Base=@n5.G) {constExponent=0.5})
  n12 = DynamicParameter() {defaultValue=<rgba 0 1 1 1>, paramNames [Variant, FlamesTurnDirection, Param3, Param4]}
  n13 = Multiply(A=Constant() {r=0.5}, B=@n12.FlamesTurnDirection) {constA=0.5}
  n14 = AppendVector(A=Add(A=Multiply(A=@n9.R), B=Multiply(A=@n11, B=@n13) {constB=0.5}), B=Multiply(A=@n9.G) {constB=0.5})
  n15 = AppendVector(A=Multiply(A=Constant() {r=0.1}, B=@n12.FlamesTurnDirection), B=Constant() {r=-0.8})
  n16 = TextureSampleParameter2D(UVs=Add(A=ParticleRandom(), B=Panner(Coordinate=@n14, Speed=@n15) {speedY=-0.8})) {
   parameterName=Tex1
   samplerType=SAMPLERTYPE_LinearGrayscale
   texture=/BRCosmetics/Weapons/FORT_Melee/Pickaxe_RenegadeRaider_Spark/FX/Materials/T_Pickaxe_RenegadeRaider_Spark_LavaNoise
  }
  n17 = LinearInterpolate(A=@n10, B=@n16.R, Alpha=ComponentMask(None=TextureCoordinate()) {g=true}) {constAlpha=0.2}
  n18 = Power(Base=@n10) {constExponent=1.3}
  n19 = VertexNormalWS()
  n20 = MaterialFunctionCall(
   Brightness=Multiply(A=@n17, B=ScalarParameter() {defaultValue=1, parameterName=Fresnel_Brightness})
   Exponent=ScalarParameter() {defaultValue=7, parameterName=Fresnel_Exponent}
   Camera Vector=CameraVectorWS()
   Vertex / Pixel Normal=@n19
  ) {materialFunction=&MF_Fresnel_FX}
  n21 = Multiply(
   A=Multiply(A=@n17, B=Saturate(None=@n18))
   B=LinearInterpolate(
    A=@n20.Fresnel Double Sided
    B=Multiply(A=@n20.Fresnel Double Sided, B=@n20.Fresnel Double Sided)
    Alpha=ScalarParameter() {parameterName=Fresnel Invert}
   ) {constAlpha=1}
  )
  n22 = Saturate(None=Power(Base=@n21) {constExponent=1})
  n23 = ParticleColor()
  MaterialFunctionCall(Emissive=Multiply(A=LinearInterpolate(A=@n2, B=@n3, Alpha=@n22), B=@n23.RGB)) {materialFunction=&MF_FX_EmissiveAdjustments}
  n24 = Power(Base=@n16.R) {constExponent=1}
  n25 = Add(A=ComponentMask(None=TextureCoordinate()) {g=true}, B=Multiply(A=Time()) {constB=-2})
  n26 = Power(Base=Sine(None=Multiply(A=@n25) {constB=0.5})) {constExponent=3}
  Multiply(A=Multiply(A=@n19, B=LinearInterpolate(A=@n24, B=@n26) {constAlpha=0.7})) {constB=3}
  n27 = Power(Base=OneMinus(None=@n5.G)) {constExponent=1}
  OneMinus(None=@n27)
  Saturate(None=@n27)
  n28 = Multiply(A=@n23.A, B=@n21)
  n29 = DepthFade(Opacity=@n28) {fadeDistanceDefault=20}
  ShadingPathSwitch(Default=@n29, Deferred=@n29, Mobile=@n28) {desc=Migrated from FeatureLevelSwitch}
  Sine(None=Multiply(A=@n9.G) {constB=0.5})
 ]
}
