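# Translucent, unlit, two-sided Niagara sprite material for flame effects on the Ripped Harvester glider.
# Twisted polar UVs pan a fire mask, distorted by scrolling Perlin noise, then combine it with radial falloff and erosion into a flame intensity.
# That intensity blends two colors, with a dynamic Variant parameter switching between warm and blue palettes, tinted by particle color.
# Opacity comes from the mask, radial fade and particle alpha.
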Material /BRCosmetics/Gadgets/Assets/VinderTech_GliderChute/Glider_RippedHarvester/FX/M_Glider_RippedHarvester_Flames
graph {
 properties {
  bUsedWithNiagaraSprites=true
  blendMode=BLEND_Translucent
  pixelDepthOffsetMode=PDOM_Legacy
  shadingModel=MSM_Unlit
  twoSided=true
  usedShadingModels=Unlit
 }
 nodes [
  n1 = MaterialFunctionCall!(UV=TextureCoordinate(), Twist Center=Constant(), Twist Edges=Constant() {r=-0.35}) {materialFunction=/Game/Accessories/FORT_Backpacks/Backpack_F_MED_Celestial_Glow/FX/Materials/MF_PolarUV}
  n2 = MaterialFunctionCall(Float2=@n1.Polar UVs) {materialFunction=&BreakOutFloat2Components}
  n3 = OneMinus(None=@n2.G)
  n4 = Power(Base=@n3) {constExponent=2.5}
  n5 = AppendVector(A=@n2.R, B=Multiply(A=Power(Base=@n2.G) {constExponent=0.666}) {constB=0.5})
  n6 = AppendVector(A=Constant(), B=Constant() {r=-0.75})
  n7 = TextureCoordinate()
  n8 = TextureSampleParameter2D(UVs=Panner(Coordinate=Multiply(A=@n7)) {speedX=-0.06, speedY=0.23}) {
   parameterName=TexDistort
   samplerType=SAMPLERTYPE_Masks
   texture=/BRCosmetics/Characters/Player/Male/Large/Bodies/M_LRG_Clash_V/Textures/Simple_Perlin_M
  }
  n9 = Multiply(A=@n8.R) {constB=0.05}
  n10 = TextureSample!(
   UVs=Add(A=Panner(Coordinate=@n5, Speed=@n6) {speedY=-0.1}, B=Multiply(A=@n2.G, B=@n9))
   Tex
   "DDX(UVs)"=@n1.DDX
   "DDY(UVs)"=@n1.DDY
   Apply View MipBias
  ) {mipValueMode=TMVM_Derivative, texture=/BRCosmetics/Athena/Effects/Particles/Skydiving/NS_Contrail_RenegadeRaider/T_FireMask}
  n11 = Saturate(None=Power(Base=OneMinus(None=@n2.G)) {constExponent=1})
  n12 = Add(A=Saturate(None=Multiply(A=@n4)), B=Multiply(A=@n10.R, B=Multiply(A=@n11)))
  n13 = OneMinus(None=@n3)
  n14 = Constant() {r=0.1}
  n15 = Multiply(
   A=MaterialFunctionCall!(Gradient=@n10.R, Erosion Amount=Power(Base=@n13) {constExponent=1}, Softness=Multiply(A=@n3, B=@n14)) {materialFunction=/BRCosmetics/Athena/Effects/Particles/Skydiving/NS_Contrail_DarkAndLight/Functions/MF_Erosion}.Increased Range
   B=@n12
  ) {constB=0.2}
  n16 = ParticleColor()
  n17 = Multiply(A=Add(A=@n12, B=Multiply(A=@n15) {constB=0.2}) {constB=0}, B=@n16.A) {constB=1.2}
  CurveAtlasRowParameter(CurveTime=Add(A=@n17) {constB=0}) {
   atlas=/Game/Weapons/FORT_Melee/Pickaxe_RippedHarvester/FX/CurveAtlas_RippedHarvester_FlameColor
   curve=/Game/Weapons/FORT_Melee/Pickaxe_RippedHarvester/FX/Curve_RippedHarvester_FireColor
   parameterName=ColorCurve
  }
  n18 = DynamicParameter() {defaultValue=<rgba 0 1 1 1>, paramNames [Variant,Param2,Param3,Param4]}
  n19 = LinearInterpolate(
   A=VectorParameter() {parameterName=Color1, defaultValue=<rgba 1 0.145 0.05 1>}.RGB
   B=VectorParameter() {parameterName=ColorAlt1, defaultValue=<rgba 0.02 0.1833 1 1>}.RGB
   Alpha=@n18.Variant
  )
  n20 = LinearInterpolate(
   A=VectorParameter() {parameterName=Color2, defaultValue=<rgba 2 1.4 0.8 1>}.RGB
   B=VectorParameter() {parameterName=ColorAlt2, defaultValue=<rgba 1 1.3333 2 1>}.RGB
   Alpha=@n18.Variant
  )
  n21 = Saturate(None=@n17)
  MaterialFunctionCall(Emissive=Multiply(A=LinearInterpolate(A=@n19, B=@n20, Alpha=@n21), B=@n16.RGB)) {materialFunction=&MF_FX_EmissiveAdjustments}
  Multiply(B=LinearInterpolate(Alpha=@n2.G) {constB=0.025}) {constA=0.01}
  n22 = Multiply(A=@n11) {constB=10}
  n23 = Add(A=@n10.R) {constB=0.5}
  Saturate(None=Multiply(A=Multiply(A=@n22, B=@n23), B=@n16.A))
 ]
}
