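# Niagara effect for the Ripped Harvester glider's engines and trails, with a red-orange default and blue alternate variant.
# Emitters sample User.SkeletalMesh bones or sockets, partitioning particles between engines for exhaust cones, heat disks, lights, flashes, embers, smoke and ribbon trails.
# Emitter-age curves stage the ignition, while the system timeline scales trail and exhaust intensity.
# An interpolated value biases trails and flashes per side, with time-of-day and front-end adjustments.
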NiagaraSystem /BRCosmetics/Gadgets/Assets/VinderTech_GliderChute/Glider_RippedHarvester/FX/NS_Glider_RippedHarvester_Trails
dependencies {
 usedDataInterfaces [
  &NiagaraDataInterfaceArrayFloat3
  &NiagaraDataInterfaceCurve
  &NiagaraDataInterfaceColorCurve
  &NiagaraDataInterfaceArrayColor
 ]
 usedDynamicInputs [&Vector2DFromFloat]
 usedModules [
  &self:SystemSpawnScript.SystemScriptSource.SystemScriptGraph.NiagaraNodeAssignment_0.SetVariables_#1
  &SystemState
  &FN_Integrator
  &Timeline
  &self:SystemSpawnScript.SystemScriptSource.SystemScriptGraph.NiagaraNodeAssignment_1.SetVariables_#2
  &InterpolateOverTime
 ]
}
emitters {
 BaseTrail {
  emitterUpdateScript [
   &EmitterState(Life Cycle Mode=::NewEnumerator1, Loop Delay=3, Loop Duration=1, UseLoopDelay=false)
   &FN_SpawnAnimTrail(
    Enable:NiagaraBool=&SetBoolByFloatComparison(A=$System.Timeline.NormalizedTime, B=0)
    Number of Anim Trails=2
    Spawn Interval=0
   ) {enabled=false}
   &FN_SpawnAnimTrail(
    Enable:NiagaraBool=&Boolean_OrOperation(
     Bool1=&SetBoolByFloatComparison(A=$System.Timeline.NormalizedTime, B=0)
     Bool2=&Boolean_AndOperation(Bool1=$User.bIsFrontEnd, Bool2=&SetBoolByFloatComparison(A=$System.Age, B=2))
    )
    Spawn Mode:ENiagaraSkeletalSpawning=::NewEnumerator0
   ) {module=FN_SpawnAnimTrail001}
  ]
  particleSpawnScript [
   &FN_InitializeAnimTrailBoneIndecies(First Bone or Socket Index=2, Second Bone or Socket Index=3)
   &FN_InitializeAnimTrail(
    AnimTrailWidth=1.8
    Color=&FN_Select_Color_from_Time_Of_Day(Morning=<rgba 1 1 1 0.7>, Day=<rgba 1 1 1 0.8>, Evening=<rgba 1 1 1 0.7>, Night=<rgba 1 1 1 0.6>)
    Lifetime=0.23
    Position Offset=<0 0 1>
   )
   &AddVelocity(Velocity=<0 -800 -70>)
  ]
  particleUpdateScript [
   &FN_UpdateAnimTrail(SkeletalMesh=$User.SkeletalMesh)
   /SetVariables(Particles.RibbonU0Override=$Particles.NormalizedAge) {isSetParameters=true, module=SetVariables_#3}
   &ScaleColor(
    Scale Alpha=&FloatFromCurve(
     FloatCurve=DataInterface() {propertyValues {Curve {keys [
         <at 0 0 RCIM_Cubic>
         {arriveTangent=-1.3016, interpMode=RCIM_Cubic, leaveTangent=-1.3016, tangentMode=RCTM_User, time=0.2, value=1}
         <at 1 0 RCIM_Cubic>
        ]}}}
     Scale Curve=&FN_FrontEndFloatModifier(
      Gameplay=&FloatFromCurve(
       FloatCurve=DataInterface() {propertyValues {Curve {keys [{arriveTangent=0.1851, interpMode=RCIM_Cubic, leaveTangent=0.1851, tangentMode=RCTM_User, time=0.5, value=0}, <at 1 1 RCIM_Cubic>]}}}
       CurveIndex=$System.Timeline.NormalizedTime
      )
      Front End=0.25
     )
    )
    Scale RGB=&MakeVectorFromLinearColorRGB(Color=&MakeCustomLinearColorFromBool(Bool=$System.IsDefaultVariant, True Color=<rgba 3 0.01 0 1>, False Color=<rgba 0 0.25 3 1>))
    ScaleA=true
    ScaleRGB=true
   )
   &ScaleColor(
    Scale Alpha=&FloatFromCurve(
     FloatCurve=DataInterface() {propertyValues {Curve {keys [<at 2 0 RCIM_Cubic>, {interpMode=RCIM_Cubic, tangentMode=RCTM_User, time=2.5, value=1}]}}}
     CurveIndex=$Emitter.Age
    )
    ScaleA=true
    ScaleRGB=false
   ) {module=ScaleColor001}
   &InheritSourceMovement(
    Applied Position Scale=&VectorFromFloat(Value=&FloatFromCurve(FloatCurve=DataInterface() {propertyValues {Curve {keys [<at 0.5 1 RCIM_Cubic>,<at 1 0.5 RCIM_Cubic>]}}}))
    ApplyForce=false
    ApplyPosition=true
    ApplyPositionFalloff=false
   ) {enabled=false}
   &SolveForcesAndVelocity(Clamp Velocity=false, Limit Acceleration=false)
  ]
  rendererClasses [&NiagaraRibbonRendererProperties]
  renderers [{rendererClass=&NiagaraRibbonRendererProperties}]
 }
 Embers {
  emitterUpdateScript [
   &EmitterState()
   &SpawnRate(
    SpawnRate=&Multiply_Float(
     A=&FloatFromCurve(
      FloatCurve=DataInterface() {propertyValues {Curve {keys [<at 0.4 0 RCIM_Cubic>,<at 1 1 RCIM_Cubic>]}}}
      CurveIndex=$System.Timeline.NormalizedTime
      Scale Curve=70
     )
     B=&FloatFromCurve(
      FloatCurve=DataInterface() {propertyValues {Curve {keys [
          {arriveTangent=0.2667, interpMode=RCIM_Cubic, time=3.2, value=0}
          {arriveTangent=-0.0798, interpMode=RCIM_Cubic, leaveTangent=-0.0798, tangentMode=RCTM_User, time=4, value=1}
         ]}}}
      CurveIndex=$Emitter.Age
     )
    )
    Use Spawn Probability=false
   )
   &SpawnBurst_Instantaneous(Spawn Count=25, Use Spawn Probability=false, UseLoopCountLimit=false) {enabled=false}
  ]
  particleSpawnScript [
   &InitializeParticle(
    Color Mode=::NewEnumerator0
    Lifetime=&RandomRangeFloat(Minimum=0.25, Maximum=0.8)
    Mass=&RandomRangeFloat(Minimum=0.5, Maximum=1)
    Mass Mode=::NewEnumerator1
    Sprite Rotation Mode=::NewEnumerator2
    Sprite Size=&Multiply_Vector2D_ByFloat(Float=&RandomRangeFloat(Minimum=2, Maximum=7), Vector2D=<0.5 1>)
    Sprite Size Mode=::NewEnumerator4
    UsePositionOffset=false
    Write Material Random=true
   )
   /SetVariables(Particles.Color=&MakeCustomLinearColorFromBool(
     Bool=$System.IsDefaultVariant
     True Color=&SelectLinearColorFromArray(
      Array Sampling Mode=::NewEnumerator0
      Direct Array Index=?
      Color Selection Array=DataInterface() {propertyValues {ColorData [<rgba 2 0.0333 0 1>, <rgba 1 0.0167 0 1>]}}
      Random Seed=0
     )
     False Color=&SelectLinearColorFromArray(
      Array Sampling Mode=::NewEnumerator0
      Direct Array Index=?
      Color Selection Array=DataInterface() {propertyValues {ColorData [<rgba 0 0.3333 2 1>, <rgba 0 0.4167 1 1>]}}
      Random Seed=0
     )
    )) {isSetParameters=true}
   &PartitionParticles()
   &GravityForce(Gravity=&RandomRangeVector(Minimum=<0 0 -300>, Maximum=<0 0 -700>)) {enabled=false, module=GravityForce001}
   &InheritVelocity(Speed Limit=true, UseSpeedThreshold=false) {enabled=false}
   &AddVelocity(
    Module.Use Falloff Distance=false
    Override Seed=false
    Velocity Mode=::NewEnumerator1
    Velocity Origin=$Particles.ShapeLocation.ShapePosition
    Velocity Speed=&RandomRangeFloat(Minimum=50, Maximum=120)
   ) {enabled=false}
   &DynamicMaterialParameters(
    Index 0 Param 1=&RandomRangeFloat(Minimum=-1, Maximum=1)
    Index 0 Param 2=&RandomRangeFloat(Minimum=0.4, Maximum=0.8)
    Module.Param0WriteEnabled=true
    Module.Param1WriteEnabled=true
    Module.Param2WriteEnabled=false
    Module.Param3WriteEnabled=false
   ) {enabled=false}
   &ApplyInitialForces() {enabled=false}
   &SkeletalMeshLocation(
    Bone / Socket Index=$Particles.PartitionParticles.Partition
    Bone Sampling Mode=::NewEnumerator4
    MaskSpawnGroup=false
    Orientation Sampling=::NewEnumerator1
    Position Sampling=::NewEnumerator2
    Sampled Position Offset=&MaskVectorBySpawnGroup(Masked Vector=<0 20 0>, Passthrough Vector=<0 -20 0>, Spawn Group Mask=$Particles.PartitionParticles.Partition)
    Skeletal Mesh=$User.SkeletalMesh
   )
   &ShapeLocation(
    MaskSpawnGroup=false
    Override Seed=false
    Shape Origin=$Particles.SkeletalMeshLocation.SampledPosition
    Sphere Radius=20
    UseOwnerScale=true
   )
   &AddVelocity(Velocity=&MaskVectorBySpawnGroup(Masked Vector=<-100 -500 -100>, Passthrough Vector=<100 -500 -100>, Spawn Group Mask=$Particles.PartitionParticles.Partition)) {module=AddVelocity001}
  ]
  particleUpdateScript [
   &ParticleState()
   &ScaleVelocity(Velocity Scale=&MakeVector(
     X=1
     Y=&FloatFromCurve(
      FloatCurve=DataInterface() {propertyValues {Curve {keys [{arriveTangent=2.0742, interpMode=RCIM_Cubic, leaveTangent=2.0742, tangentMode=RCTM_User, time=0, value=0.1}, <at 1 1 RCIM_Cubic>]}}}
      Scale Curve=2
     )
     Z=1
    ))
   &GravityForce(Gravity=<0 0 0>) {enabled=false}
   &CurlNoiseForce(Noise Frequency=100, Noise Quality / Cost=::NewEnumerator0, Noise Strength=800, Pan Noise=false)
   &Drag(Drag=4, Use Linear Drag=true, Use Rotational Drag=true) {enabled=false}
   &SolveForcesAndVelocity(Clamp Velocity=false, Limit Acceleration=false)
   &ScaleSpriteSize(Uniform Curve Sprite Scale=DataInterface() {propertyValues {Curve {keys [
        {arriveTangent=6.8904, interpMode=RCIM_Cubic, leaveTangent=6.8904, tangentMode=RCTM_User, time=0, value=0}
        {arriveTangent=10, interpMode=RCIM_Cubic, leaveTangent=10, time=0.0151, value=0.25}
        {arriveTangentWeight=0.0745, interpMode=RCIM_Cubic, leaveTangentWeight=0.2981, tangentMode=RCTM_User, time=0.1, value=1}
        {arriveTangent=-1.2657, interpMode=RCIM_Cubic, leaveTangent=-1.2657, time=0.2913, value=0.1}
        {interpMode=RCIM_Cubic, tangentMode=RCTM_User, time=0.337, value=0.7}
        {arriveTangent=-0.5124, interpMode=RCIM_Cubic, leaveTangent=-0.5124, tangentMode=RCTM_User, time=0.5849, value=0.15}
        {tangentMode=RCTM_User, time=1, value=0}
       ]}}})
   &ScaleSpriteSizeBySpeed(Max Scale Factor=<1 2>, Velocity Threshold=500)
   &ScaleColor(
    Scale Alpha=&FloatFromCurve(
     FloatCurve=DataInterface() {propertyValues {Curve {keys [
         {arriveTangent=9.0725, interpMode=RCIM_Cubic, leaveTangent=9.0725, tangentMode=RCTM_User, time=0, value=0}
         {arriveTangent=-2.5249, interpMode=RCIM_Cubic, leaveTangent=-2.5249, tangentMode=RCTM_User, time=0.1124, value=0.9847}
         {arriveTangent=-0.1423, interpMode=RCIM_Cubic, leaveTangent=-0.1423, tangentMode=RCTM_User, time=0.2564, value=0.4594}
         {arriveTangent=2.0872, interpMode=RCIM_Cubic, leaveTangent=2.0872, time=0.4, value=0.4}
         {arriveTangent=-0.5, interpMode=RCIM_Cubic, leaveTangent=-0.5, time=0.4332, value=0.8284}
         {arriveTangent=-1.3751, interpMode=RCIM_Cubic, leaveTangent=-1.3751, time=0.6, value=0.3}
         {arriveTangent=-4.3712, interpMode=RCIM_Cubic, leaveTangent=-4.3712, tangentMode=RCTM_User, time=0.6352, value=0.5506}
         {arriveTangent=-0.0405, interpMode=RCIM_Cubic, leaveTangent=-0.0405, tangentMode=RCTM_User, time=1, value=0}
        ]}}}
     Scale Curve=20
    )
    ScaleA=true
    ScaleRGB=true
   )
  ]
  rendererClasses [&NiagaraSpriteRendererProperties]
  renderers [{rendererClass=&NiagaraSpriteRendererProperties}]
 }
 EngineLight {
  emitterUpdateScript [
   &EmitterState()
   &SpawnBurst_Instantaneous(Spawn Count=4, Use Spawn Probability=false, UseLoopCountLimit=true)
  ]
  particleSpawnScript [
   &InitializeParticle(
    Color=&MakeCustomLinearColorFromBool(Bool=$System.IsDefaultVariant, True Color=<rgba 42 1.113 0.42 1>, False Color=<rgba 0.25 6.4375 25 1>)
    Sprite Size Mode=::NewEnumerator3
    Uniform Sprite Size=115
    UsePositionOffset=false
    Write Material Random=true
   )
   &PartitionParticles()
   &CameraOffset(Camera Offset Amount=5) {enabled=false}
  ]
  particleUpdateScript [
   &SkeletalMeshLocation(
    Bone / Socket Index=$Particles.PartitionParticles.Partition
    Bone Sampling Mode=::NewEnumerator4
    MaskSpawnGroup=false
    Sampled Position Offset=&MaskVectorBySpawnGroup(Masked Vector=<0 0 0>, Passthrough Vector=<0 0 0>, Spawn Group Mask=$Particles.PartitionParticles.Partition)
    Skeletal Mesh=$User.SkeletalMesh
   )
   &ParticleState(Kill Particles When Lifetime Has Elapsed=false, Let Infinitely Lived Particles Die When Emitter Deactivates=true)
   &ScaleColor(
    Scale Alpha=&FloatFromCurve(
     FloatCurve=DataInterface() {propertyValues {Curve {keys [<at 0 0 RCIM_Cubic>,<at 1 1 RCIM_Cubic>]}}}
     CurveIndex=$System.Timeline.NormalizedTime
     Scale Curve=&FN_Select_Float_from_Time_Of_Day(Morning=2, Day=2, Evening=2, Night=1.5)
    )
    ScaleA=true
    ScaleRGB=false
   )
   &ScaleColor(
    Scale Alpha=&FloatFromCurve(FloatCurve=DataInterface() {propertyValues {Curve {keys [<at 3.5 0 RCIM_Cubic>,<at 4.5 1 RCIM_Cubic>]}}}, CurveIndex=$Emitter.Age)
    ScaleA=true
    ScaleRGB=false
   ) {module=ScaleColor001}
  ]
  rendererClasses [&NiagaraSpriteRendererProperties, &NiagaraLightRendererProperties]
  renderers [
   {rendererClass=&NiagaraSpriteRendererProperties}
   {rendererClass=&NiagaraLightRendererProperties, rendererIndex=1}
  ]
 }
 FlashBurstDeploy {
  emitterUpdateScript [
   &EmitterState(Life Cycle Mode=::NewEnumerator1, Loop Behavior=::NewEnumerator1, Loop Delay=3.16, Loop Duration=1, UseLoopDelay=true)
   &SpawnBurst_Instantaneous(Spawn Count=4, Use Spawn Probability=false, UseLoopCountLimit=true)
  ]
  particleSpawnScript [
   &InitializeParticle(
    Color=&MakeCustomLinearColorFromBool(Bool=$System.IsDefaultVariant, True Color=<rgba 6 0.3 0 1>, False Color=<rgba 0.03 0.1785 3 1>)
    Lifetime=1.8
    Sprite Size=&Multiply_Vector2D_ByFloat(Float=&FN_FrontEndFloatModifier(Gameplay=1.3, Front End=1), Vector2D=<130 80>)
    Sprite Size Mode=::NewEnumerator4
    UsePositionOffset=false
    Write Material Random=true
   )
   &PartitionParticles()
   &AddVelocity(Velocity=<0 0 300>) {enabled=false}
  ]
  particleUpdateScript [
   &SkeletalMeshLocation(
    Bone / Socket Index=$Particles.PartitionParticles.Partition
    Bone Sampling Mode=::NewEnumerator4
    MaskSpawnGroup=false
    Orientation Sampling=::NewEnumerator1
    Sampled Position Offset=&MaskVectorBySpawnGroup(Masked Vector=<0 10 0>, Passthrough Vector=<0 -10 0>, Spawn Group Mask=$Particles.PartitionParticles.Partition)
    Skeletal Mesh=$User.SkeletalMesh
   )
   &ParticleState()
   &ScaleSpriteSize(Uniform Curve Sprite Scale=DataInterface() {propertyValues {Curve {keys [
        {arriveTangent=0.1509, interpMode=RCIM_Cubic, leaveTangent=0.1509, tangentMode=RCTM_User, time=0, value=0.2}
        {arriveTangent=0.8442, interpMode=RCIM_Cubic, leaveTangent=0.8442, tangentMode=RCTM_User, time=0.15, value=0.8}
        <at 0.6 1 RCIM_Cubic>
       ]}}})
   &ScaleColor(
    Scale Alpha=&FloatFromCurve(
     FloatCurve=DataInterface() {propertyValues {Curve {keys [
         {arriveTangent=0.5671, interpMode=RCIM_Cubic, leaveTangent=0.5671, tangentMode=RCTM_User, time=0, value=0}
         {arriveTangent=-0.9598, interpMode=RCIM_Cubic, leaveTangent=-0.9598, tangentMode=RCTM_User, time=0.1, value=1}
         {arriveTangent=-0.6531, interpMode=RCIM_Cubic, leaveTangent=-0.6531, tangentMode=RCTM_User, time=1, value=0}
        ]}}}
     Scale Curve=&FN_FrontEndFloatModifier(Gameplay=1.5, Front End=0.3)
    )
    Scale RGB=&MakeVectorFromLinearColorRGB(Color=<rgba 1 1 1 1>)
    ScaleA=true
    ScaleRGB=false
   ) {module=ScaleColor002}
  ]
  rendererClasses [&NiagaraSpriteRendererProperties]
  renderers [{rendererClass=&NiagaraSpriteRendererProperties}]
 }
 FlashBurstForward {
  emitterUpdateScript [
   &EmitterState(Life Cycle Mode=::NewEnumerator1, Loop Delay=3.4, Loop Duration=0, UseLoopDelay=false)
   &SpawnBurst_Instantaneous(
    Loop Count Limit=2
    Spawn Count=4
    Spawn Probability=HlslExpression() {hlslExpression="Emitter.LoopCount == 0"}
    Use Spawn Probability=false
    UseLoopCountLimit=true
   )
  ]
  particleSpawnScript [
   &PartitionParticles()
   &InitializeParticle(
    Color=&MakeCustomLinearColorFromBool(Bool=$System.IsDefaultVariant, True Color=<rgba 4 0.238 0.04 1>, False Color=<rgba 0.03 0.1785 3 0.5>)
    Sprite Size=<160 80>
    Sprite Size Mode=::NewEnumerator4
    UsePositionOffset=false
    Write Material Random=true
   )
  ]
  particleUpdateScript [
   &SkeletalMeshLocation(
    Bone / Socket Index=$Particles.PartitionParticles.Partition
    Bone Sampling Mode=::NewEnumerator4
    MaskSpawnGroup=false
    Orientation Sampling=::NewEnumerator1
    Sampled Position Offset=&MaskVectorBySpawnGroup(Masked Vector=<0 8 0>, Passthrough Vector=<0 -8 0>, Spawn Group Mask=$Particles.PartitionParticles.Partition)
    Skeletal Mesh=$User.SkeletalMesh
   )
   &CameraOffset(Camera Offset Amount=20)
   &ParticleState(
    Kill Particles When Lifetime Has Elapsed=false
    Let Infinitely Lived Particles Die When Emitter Deactivates=true
    Loop Particles Lifetime=true
   )
   &SpriteRotationRate(Rotation Rate=4) {enabled=false}
   &ScaleSpriteSize(
    Uniform Curve Index=$System.Timeline.NormalizedTime
    Uniform Curve Sprite Scale=DataInterface() {propertyValues {Curve {keys [{arriveTangent=0.06, interpMode=RCIM_Cubic, leaveTangent=0.06, tangentMode=RCTM_User, time=0.2, value=0}, <at 1 1 RCIM_Cubic>]}}}
   )
   &ScaleSpriteSize(Uniform Curve Sprite Scale=DataInterface() {propertyValues {Curve {keys [
        <at 0 1 RCIM_Cubic>
        {arriveTangent=-0.5, interpMode=RCIM_Cubic, leaveTangent=-0.5, time=0.1, value=0.85}
        {arriveTangent=0.5, interpMode=RCIM_Cubic, leaveTangent=0.5, time=0.2, value=0.9}
        <at 0.3 0.95 RCIM_Cubic>
        {arriveTangent=-1, interpMode=RCIM_Cubic, leaveTangent=-1, time=0.4, value=0.9}
        {arriveTangent=-0.25, interpMode=RCIM_Cubic, leaveTangent=-0.25, time=0.5, value=0.75}
        {arriveTangent=1, interpMode=RCIM_Cubic, leaveTangent=1, time=0.6, value=0.85}
        {arriveTangent=0.25, interpMode=RCIM_Cubic, leaveTangent=0.25, time=0.7, value=0.95}
        {arriveTangent=0.5, interpMode=RCIM_Cubic, leaveTangent=0.5, time=0.8, value=0.9}
        {arriveTangent=0.5, interpMode=RCIM_Cubic, leaveTangent=0.5, time=0.9, value=1.05}
        <at 1 1 RCIM_Cubic>
       ]}}}) {module=ScaleSpriteSize001}
   &ScaleColor(
    Scale Alpha=&FloatFromCurve(
     FloatCurve=DataInterface() {propertyValues {Curve {keys [<at 0.5 0 RCIM_Cubic>,<at 1 1 RCIM_Cubic>]}}}
     CurveIndex=$System.Timeline.NormalizedTime
     Scale Curve=&FN_Select_Float_from_Time_Of_Day(Morning=1, Day=1, Evening=1, Night=1)
    )
    ScaleA=true
    ScaleRGB=false
   )
   &ScaleColor(
    Scale Alpha=&FloatFromCurve(FloatCurve=DataInterface() {propertyValues {Curve {keys [<at 3.5 0 RCIM_Cubic>,<at 4.5 1 RCIM_Cubic>]}}}, CurveIndex=$Emitter.Age)
    ScaleA=true
    ScaleRGB=false
   ) {module=ScaleColor001}
   &SolveForcesAndVelocity(Clamp Velocity=false, Limit Acceleration=false)
   &ScaleColor(
    Scale Alpha=&FloatFromCurve(
     FloatCurve=DataInterface() {propertyValues {Curve {keys [
         {arriveTangent=16.9192, interpMode=RCIM_Cubic, leaveTangent=16.9192, tangentMode=RCTM_User, time=0, value=0}
         {arriveTangent=-1.2434, interpMode=RCIM_Cubic, leaveTangent=-1.2434, tangentMode=RCTM_User, time=0.1, value=1}
         {arriveTangent=-0.6531, interpMode=RCIM_Cubic, leaveTangent=-0.6531, tangentMode=RCTM_User, time=1, value=0}
        ]}}}
     Scale Curve=1.5
    )
    Scale RGB=&MakeVectorFromLinearColorRGB(Color=<rgba 1 1 1 1>)
    ScaleA=true
    ScaleRGB=false
   ) {enabled=false, module=ScaleColor002}
   &ScaleColor(
    Scale Alpha=&MaskFloatBySpawnGroup(
     Masked Float=&Add_Float(A=1, B=&Multiply_Float(A=$System.InterpolateOverTime.InterpolateOverTime_Float.Moving Average, B=0.5))
     Passthrough Float=&Subtract_Float(A=1, B=&Multiply_Float(A=$System.InterpolateOverTime.InterpolateOverTime_Float.Moving Average, B=0.5))
     Spawn Group Mask=$Particles.PartitionParticles.Partition
    )
    ScaleA=true
    ScaleRGB=false
   ) {module=ScaleColor003}
  ]
  rendererClasses [&NiagaraSpriteRendererProperties]
  renderers [{rendererClass=&NiagaraSpriteRendererProperties}]
 }
 HeatCones {
  emitterUpdateScript [
   &EmitterState(Life Cycle Mode=::NewEnumerator1, Loop Behavior=::NewEnumerator1, Loop Delay=3, Loop Duration Mode=::NewEnumerator1, UseLoopDelay=false)
   &SpawnBurst_Instantaneous(Spawn Count=4, Use Spawn Probability=false, UseLoopCountLimit=true)
  ]
  particleSpawnScript [
   &InitializeParticle(
    Lifetime=7
    Mesh Scale=&Multiply_VectorByFloat(Vector=<1 1 4>, Float=&FN_FrontEndFloatModifier(Gameplay=1, Front End=0.6))
    Mesh Scale Mode=::NewEnumerator4
    UsePositionOffset=false
    Write Material Random=true
   )
   &PartitionParticles()
   /SetVariables(Particles.Position=&AddVectorToPosition(Position=$Particles.Position, Vector=<0 0 -5>)) {enabled=false, isSetParameters=true, module=SetVariables_#4}
   &AddVelocity() {enabled=false}
   &DynamicMaterialParameters(
    Index 0 Param 1=$User.Variant
    Index 0 Param 2=&MaskFloatBySpawnGroup(Masked Float=-1, Passthrough Float=1, Spawn Group Mask=$Particles.PartitionParticles.Partition)
    Module.Param0WriteEnabled=true
    Module.Param1WriteEnabled=true
    Module.Param2WriteEnabled=false
    Module.Param3WriteEnabled=false
   )
  ]
  particleUpdateScript [
   &SkeletalMeshLocation(
    Bone / Socket Index=&ModuloInt(A=$Particles.PartitionParticles.Partition, B=2)
    Bone Sampling Mode=::NewEnumerator4
    MaskSpawnGroup=false
    Orientation Sampling=::NewEnumerator1
    Sampled Position Offset=&SelectVectorFromArray(
     Direct Array Index=$Particles.PartitionParticles.Partition
     Direct Array Mode=::NewEnumerator1
     Vector Selection Array=DataInterface() {propertyValues {FloatData [<0 0 0>, <0 0 0>]}}
    )
    Skeletal Mesh=$User.SkeletalMesh
   )
   /SetVariables(Particles.Color=&FN_Select_Color_from_Time_Of_Day(Morning=<rgba 0.8 0.8 0.8 1>, Day=<rgba 1 1 1 1>, Evening=<rgba 0.8 0.8 0.8 1>, Night=<rgba 0.6 0.6 0.6 1>)) {isSetParameters=true}
   &ParticleState(
    Kill Particles When Lifetime Has Elapsed=false
    Let Infinitely Lived Particles Die When Emitter Deactivates=true
    Loop Particles Lifetime=true
   )
   /SetVariables(Particles.MeshOrientation=&MultiplyQuaternion(
     Quaternion A=$Particles.SkeletalMeshLocation.SampledRotation
     Quaternion B=&MakeQuatFromAxisAngle(
      Axis=&MaskVectorBySpawnGroup(Masked Vector=<1 0 0>, Passthrough Vector=<-1 0 0>, Spawn Group Mask=$Particles.PartitionParticles.Partition)
      AngleInDegrees=90
     )
    )) {isSetParameters=true}
   /SetVariables(Particles.MeshOrientation=&MultiplyQuaternion(
     Quaternion A=$Particles.MeshOrientation
     Quaternion B=/BRCosmetics/Athena/Effects/Particles/Skydiving/NS_Contrail_Apprentice/OrientMestToVectorFromCustomRot(
      InitialRotation:Quat4f=<0 0 0 1>
      Look At Direction:Vector3f=<0 0 0>
      Reference Axis Vector:Vector3f=<0 0 0>
      RotationVector:Vector3f=&MakeVector(
       X=0
       Y=0
       Z=&FloatFromCurve(FloatCurve=DataInterface() {propertyValues {Curve {keys [<at 0 0>,<at 1 1 RCIM_Cubic>]}}})
      )
     )
    )) {enabled=false, isSetParameters=true}
   &ScaleMeshSize(Scale Factor=&MaskVectorBySpawnGroup(Masked Vector=<-1 1 1>, Passthrough Vector=<1 1 1>, Spawn Group Mask=$Particles.PartitionParticles.Partition)) {enabled=false, module=ScaleMeshSize001}
   &ScaleMeshSize(Scale Factor=&MakeVector(
     X=1
     Y=1
     Z=&Multiply_Float(
      A=&FloatFromCurve(FloatCurve=DataInterface() {propertyValues {Curve {keys [
           <at 0 1 RCIM_Cubic>
           {arriveTangent=-0.75, interpMode=RCIM_Cubic, leaveTangent=-0.75, time=0.1, value=0.65}
           {arriveTangent=1, interpMode=RCIM_Cubic, leaveTangent=1, time=0.2, value=0.85}
           {arriveTangent=1.25, interpMode=RCIM_Cubic, leaveTangent=1.25, time=0.3, value=0.85}
           {arriveTangent=0.25, interpMode=RCIM_Cubic, leaveTangent=0.25, time=0.4, value=1.1}
           {arriveTangent=-0.5, interpMode=RCIM_Cubic, leaveTangent=-0.5, time=0.5, value=0.9}
           {arriveTangent=-0.75, interpMode=RCIM_Cubic, leaveTangent=-0.75, time=0.6, value=1}
           {arriveTangent=-2, interpMode=RCIM_Cubic, leaveTangent=-2, time=0.7, value=0.75}
           {arriveTangent=1, interpMode=RCIM_Cubic, leaveTangent=1, time=0.8, value=0.6}
           {arriveTangent=2, interpMode=RCIM_Cubic, leaveTangent=2, time=0.9, value=0.95}
           <at 1 1 RCIM_Cubic>
          ]}}})
      B=&FloatFromCurve(
       FloatCurve=DataInterface() {propertyValues {Curve {keys [
           <at 0 1 RCIM_Cubic>
           {arriveTangent=-0.5, interpMode=RCIM_Cubic, leaveTangent=-0.5, time=0.1, value=0.85}
           {arriveTangent=0.5, interpMode=RCIM_Cubic, leaveTangent=0.5, time=0.2, value=0.9}
           <at 0.3 0.95 RCIM_Cubic>
           {arriveTangent=-1, interpMode=RCIM_Cubic, leaveTangent=-1, time=0.4, value=0.9}
           {arriveTangent=-0.25, interpMode=RCIM_Cubic, leaveTangent=-0.25, time=0.5, value=0.75}
           {arriveTangent=1, interpMode=RCIM_Cubic, leaveTangent=1, time=0.6, value=0.85}
           {arriveTangent=0.25, interpMode=RCIM_Cubic, leaveTangent=0.25, time=0.7, value=0.95}
           {arriveTangent=0.5, interpMode=RCIM_Cubic, leaveTangent=0.5, time=0.8, value=0.9}
           {arriveTangent=0.5, interpMode=RCIM_Cubic, leaveTangent=0.5, time=0.9, value=1.05}
           <at 1 1 RCIM_Cubic>
          ]}}}
       CurveIndex=&Multiply_Float(A=$Particles.NormalizedAge, B=2)
      )
     )
    ))
   &ScaleMeshSize(Scale Factor=&MakeVector(
     X=1
     Y=1
     Z=&FloatFromCurve(
      FloatCurve=DataInterface() {propertyValues {Curve {keys [
          {arriveTangent=0.2667, interpMode=RCIM_Cubic, time=3.5, value=0}
          {arriveTangent=-0.0798, interpMode=RCIM_Cubic, leaveTangent=-0.0798, tangentMode=RCTM_User, time=4, value=1.5}
          <at 5 1 RCIM_Cubic>
         ]}}}
      CurveIndex=$Emitter.Age
     )
    )) {module=ScaleMeshSize002}
   &SolveForcesAndVelocity(Clamp Velocity=false, Limit Acceleration=false)
   /SetVariables(Particles.ColorInnerExhaust:LinearColor=&MakeLinearColorFromVectorAndFloat(
     "Vector (RGB)"=&MakeVectorFromLinearColorRGB(Color=&MultiplyLinearColor(
       Linear Color=&MakeCustomLinearColorFromBool(Bool=$System.IsDefaultVariant, True Color=<rgba 4 1.44 0.8 1>, False Color=<rgba 0.28 1.21 4 1>)
       Scale Factor=&MakeVector4_FromLinearColor(LinearColor=$Particles.Color)
      ))
     "Float (Alpha)"=&FloatFromCurve(FloatCurve=DataInterface() {propertyValues {Curve {keys [<at 0 1>,<at 1 3>]}}}, CurveIndex=$System.Timeline.NormalizedTime)
    )) {isSetParameters=true}
   /SetVariables(Particles.ColorOuterExhaust:LinearColor=&MakeLinearColorFromVectorAndFloat(
     "Vector (RGB)"=&MakeVectorFromLinearColorRGB(Color=&MultiplyLinearColor(
       Linear Color=&MakeCustomLinearColorFromBool(Bool=$System.IsDefaultVariant, True Color=<rgba 30 1.785 0.3 1>, False Color=<rgba 0.3 10.2 30 1>)
       Scale Factor=&MakeVector4_FromLinearColor(LinearColor=$Particles.Color)
      ))
     "Float (Alpha)"=&FloatFromCurve(
      FloatCurve=DataInterface() {propertyValues {Curve {keys [{arriveTangent=-2.916, interpMode=RCIM_Cubic, leaveTangent=-2.916, tangentMode=RCTM_User, time=0, value=1}, <at 0.8 0 RCIM_Cubic>]}}}
      CurveIndex=$System.Timeline.NormalizedTime
      Scale Curve=0.2
     )
    )) {isSetParameters=true}
   /SetVariables(Particles.ColorUpperExhaust:LinearColor=&MakeLinearColorFromVectorAndFloat(
     "Vector (RGB)"=&MakeVectorFromLinearColorRGB(Color=&MultiplyLinearColor(
       Linear Color=&MakeCustomLinearColorFromBool(Bool=$System.IsDefaultVariant, True Color=<rgba 1 0.875 0.7 1>, False Color=<rgba 0.7 0.8 1 1>)
       Scale Factor=&MakeVector4_FromLinearColor(LinearColor=$Particles.Color)
      ))
     "Float (Alpha)"=&FloatFromCurve(
      FloatCurve=DataInterface() {propertyValues {Curve {keys [{arriveTangent=-2.916, interpMode=RCIM_Cubic, leaveTangent=-2.916, tangentMode=RCTM_User, time=0, value=1}, <at 0.8 0 RCIM_Cubic>]}}}
      CurveIndex=$System.Timeline.NormalizedTime
      Scale Curve=0.2
     )
    )) {enabled=false, isSetParameters=true}
   &ScaleColor(
    Scale Alpha=&FloatFromCurve(
     FloatCurve=DataInterface() {propertyValues {Curve {keys [{arriveTangent=-2.916, interpMode=RCIM_Cubic, leaveTangent=-2.916, tangentMode=RCTM_User, time=0, value=1}, <at 0.8 0 RCIM_Cubic>]}}}
     CurveIndex=$System.Timeline.NormalizedTime
     Scale Curve=0.1
    )
    ScaleA=true
    ScaleRGB=false
   ) {enabled=false}
  ]
  rendererClasses [&NiagaraMeshRendererProperties]
  renderers [
   {rendererClass=&NiagaraMeshRendererProperties, rendererIndex=0}
   {rendererClass=&NiagaraMeshRendererProperties, rendererIndex=1}
   {rendererClass=&NiagaraMeshRendererProperties, rendererIndex=2}
  ]
 }
 HeatDisks {
  emitterUpdateScript [
   &EmitterState(Life Cycle Mode=::NewEnumerator1, Loop Behavior=::NewEnumerator1, Loop Duration Mode=::NewEnumerator1, UseLoopDelay=false)
   &SpawnBurst_Instantaneous(Spawn Count=8, Use Spawn Probability=false, UseLoopCountLimit=true)
  ]
  particleSpawnScript [
   &InitializeParticle(
    Color=&MakeCustomLinearColorFromBool(Bool=$System.IsDefaultVariant, True Color=<rgba 1 0.725 0.7 1>, False Color=<rgba 0.7 0.75 1 1>)
    Lifetime=8
    Sprite Size Mode=::NewEnumerator3
    Uniform Sprite Size=&FN_FrontEndFloatModifier(Gameplay=65, Front End=36)
    UsePositionOffset=false
    Write Material Random=true
   )
   &PartitionParticles(Number Of Partitions=4)
   /SetVariables(Particles.Position=&AddVectorToPosition(Position=$Particles.Position, Vector=<0 0 -5>)) {enabled=false, isSetParameters=true, module=SetVariables_#4}
   &DynamicMaterialParameters(
    Index 0 Param 1=$User.Variant
    Module.Param0WriteEnabled=true
    Module.Param1WriteEnabled=false
    Module.Param2WriteEnabled=false
    Module.Param3WriteEnabled=false
   )
  ]
  particleUpdateScript [
   &SkeletalMeshLocation(
    Bone / Socket Index=&ModuloInt(A=$Particles.PartitionParticles.Partition, B=2)
    Bone Sampling Mode=::NewEnumerator4
    MaskSpawnGroup=false
    Orientation Sampling=::NewEnumerator1
    Sampled Position Offset=&SelectVectorFromArray(
     Direct Array Index=$Particles.PartitionParticles.Partition
     Direct Array Mode=::NewEnumerator1
     Vector Selection Array=DataInterface() {propertyValues {FloatData [<0 -1.5 0>, <0 1.5 0>, <0 1.5 0>, <0 -1.5 0>]}}
    )
    Skeletal Mesh=$User.SkeletalMesh
   )
   &SpriteFacingAndAlignment(
    Sprite Facing=&RotateVectorByQuaternion(Vector To Rotate=<0 1 0>, Quaternion=$Particles.SkeletalMeshLocation.SampledRotation)
    WriteAlignment=false
    WriteFacing=true
   )
   &ParticleState(
    Kill Particles When Lifetime Has Elapsed=false
    Let Infinitely Lived Particles Die When Emitter Deactivates=true
    Loop Particles Lifetime=true
   )
   &ScaleSpriteSize(
    Uniform Curve Index=$Emitter.Age
    Uniform Curve Sprite Scale=DataInterface() {propertyValues {Curve {keys [<at 1.5 0 RCIM_Cubic>,<at 3 1 RCIM_Cubic>]}}}
   )
   /SetVariables(Particles.SpriteRotationOuter:NiagaraFloat=&Multiply_Float(A=$Particles.SpriteRotation, B=1.5)) {isSetParameters=true}
   &SolveForcesAndVelocity(Clamp Velocity=false, Limit Acceleration=false)
   &ScaleColor(
    Scale Alpha=&FloatFromCurve(
     FloatCurve=DataInterface() {propertyValues {Curve {keys [<at 0 1 RCIM_Cubic>,<at 0.25 1.2 RCIM_Cubic>,<at 0.5 1 RCIM_Cubic>,<at 0.75 1.2 RCIM_Cubic>,<at 1 1 RCIM_Cubic>]}}}
     Scale Curve=&FloatFromCurve(
      FloatCurve=DataInterface() {propertyValues {Curve {keys [<at 0 1 RCIM_Cubic>,<at 1 1.5 RCIM_Cubic>]}}}
      CurveIndex=$System.Timeline.NormalizedTime
      Scale Curve=&FN_Select_Float_from_Time_Of_Day(Morning=0.7, Day=1, Evening=0.7, Night=0.5)
     )
    )
    Scale RGB=&VectorFromFloat(Value=&Lerp_Float(A=0.8, B=2.5, Alpha=$System.Timeline.NormalizedTime))
    ScaleA=true
    ScaleRGB=true
   ) {enabled=false}
   &ScaleColor(
    Scale Alpha=&FloatFromCurve(
     FloatCurve=DataInterface() {propertyValues {Curve {keys [
         <at 3.4 0.5 RCIM_Cubic>
         {arriveTangent=-0.0638, interpMode=RCIM_Cubic, leaveTangent=-0.0638, tangentMode=RCTM_User, time=3.8, value=4}
         {arriveTangent=-0.0372, interpMode=RCIM_Cubic, leaveTangent=-0.0372, tangentMode=RCTM_User, time=4.3, value=4}
         <at 6 1 RCIM_Cubic>
        ]}}}
     CurveIndex=$Emitter.Age
    )
    Scale RGB=&MakeCustomVectorFromBool(
     Bool=$System.IsDefaultVariant
     True Vector=&MakeVectorFromLinearColorRGB(Color=&ColorFromCurve(
       ColorCurve=DataInterface() {propertyValues {
         AlphaCurve {keys []}
         BlueCurve {keys [<at 0 1 RCIM_Cubic>,<at 0.9536 0.2 RCIM_Cubic>]}
         GreenCurve {keys [<at 0 1 RCIM_Cubic>,<at 0.9536 0.2667 RCIM_Cubic>]}
         RedCurve {keys [<at 0 1 RCIM_Cubic>,<at 0.9536 1 RCIM_Cubic>]}
        }}
       CurveIndex=$System.Timeline.NormalizedTime
      ))
     False Vector=&MakeVectorFromLinearColorRGB(Color=&ColorFromCurve(
       ColorCurve=DataInterface() {propertyValues {
         AlphaCurve {keys []}
         BlueCurve {keys [<at 0 1 RCIM_Cubic>,<at 0.9536 1 RCIM_Cubic>]}
         GreenCurve {keys [<at 0 1 RCIM_Cubic>,<at 0.9536 0.4 RCIM_Cubic>]}
         RedCurve {keys [<at 0 1 RCIM_Cubic>,<at 0.9536 0.2 RCIM_Cubic>]}
        }}
       CurveIndex=$System.Timeline.NormalizedTime
      ))
    )
    ScaleA=true
    ScaleRGB=true
   ) {module=ScaleColor002}
   &ScaleColor(
    Scale Alpha=&FN_Select_Float_from_Time_Of_Day(Morning=0.8, Day=1, Evening=0.8, Night=0.8)
    Scale RGB=&VectorFromFloat(Value=&FloatFromCurve(
      FloatCurve=DataInterface() {propertyValues {Curve {keys [<at 0 1 RCIM_Cubic>,<at 0.25 1.2 RCIM_Cubic>,<at 0.5 1 RCIM_Cubic>,<at 0.75 1.2 RCIM_Cubic>,<at 1 1 RCIM_Cubic>]}}}
      Scale Curve=&Lerp_Float(
       A=0.9
       B=&MakeCustomFloatFromBool(
        Bool=$System.IsDefaultVariant
        True Float=5
        False Float=&FN_Select_Float_from_Time_Of_Day(Morning=3, Day=3.6, Evening=3, Night=1.3)
       )
       Alpha=$System.Timeline.NormalizedTime
      )
     ))
    ScaleA=false
    ScaleRGB=true
   ) {module=ScaleColor001}
  ]
  rendererClasses [&NiagaraSpriteRendererProperties]
  renderers [{rendererClass=&NiagaraSpriteRendererProperties}]
 }
 Main_Trail {
  emitterUpdateScript [
   &EmitterState(Life Cycle Mode=::NewEnumerator1, Loop Delay=3, Loop Duration=1, UseLoopDelay=false)
   &FN_SpawnAnimTrail(
    Enable:NiagaraBool=&SetBoolByFloatComparison(A=$System.Timeline.NormalizedTime, B=0)
    Number of Anim Trails=2
    Spawn Interval=0
   ) {enabled=false}
   &FN_SpawnAnimTrail(Number of Anim Trails=2) {module=FN_SpawnAnimTrail001}
  ]
  particleSpawnScript [
   &FN_InitializeAnimTrailBoneIndecies()
   &FN_InitializeAnimTrailBoneIndecies(AnimTrail Selection=1, First Bone or Socket Index=2, Second Bone or Socket Index=3) {module=FN_InitializeAnimTrailBoneIndecies001}
   &FN_InitializeAnimTrail(
    AnimTrailWidth=1.15
    Color=&FN_Select_Color_from_Time_Of_Day(Morning=<rgba 1 1 1 0.8>, Day=<rgba 1 1 1 1>, Evening=<rgba 1 1 1 0.8>, Night=<rgba 1 1 1 0.3>)
   )
   &DynamicMaterialParameters(
    Index 0 Param 1=$User.Variant
    Module.Param0WriteEnabled=true
    Module.Param1WriteEnabled=false
    Module.Param2WriteEnabled=false
    Module.Param3WriteEnabled=false
   )
  ]
  particleUpdateScript [
   &SkeletalMeshLocation(
    Bone / Socket Index=$Particles.SpawnGroup
    Bone Sampling Mode=::NewEnumerator4
    MaskSpawnGroup=false
    Orientation Sampling=::NewEnumerator1
    Skeletal Mesh=$User.SkeletalMesh
   )
   /SetVariables(Particles.Position=&AddVectorToPosition(
     Position=$Particles.Position
     Vector=&Multiply_VectorByFloat(
      Vector=&RotateVectorByQuaternion(
       Vector To Rotate=&MakeVector(
        X=0
        Y=&ClampFloat(
         Float=&Multiply_Float(
          A=&Lerp_Float(A=-2, B=2, Alpha=$System.Timeline.NormalizedTime)
          B=&Add_Float(
           A=&MaskFloatBySpawnGroup(Masked Float=1, Passthrough Float=-1)
           B=&Multiply_Float(A=$System.InterpolateOverTime.InterpolateOverTime_Float.Moving Average, B=1)
          )
         )
         Min=-1.5
         Max=1.5
        )
        Z=0
       )
       Quaternion=$Particles.SkeletalMeshLocation.SampledRotation
      )
      Float=&FloatFromCurve(
       FloatCurve=DataInterface() {propertyValues {Curve {keys [<at 0 0>,<at 1 1>]}}}
       Scale Curve=&FloatFromCurve(
        FloatCurve=DataInterface() {propertyValues {Curve {keys [
            <at 0.15 0.8 RCIM_Cubic>
            {arriveTangent=0.0007, interpMode=RCIM_Cubic, leaveTangent=0.0007, tangentMode=RCTM_User, time=0.35, value=1.5}
            {interpMode=RCIM_Cubic, leaveTangent=-0.4242, time=1, value=0.8}
           ]}}}
        CurveIndex=$System.Timeline002.NormalizedTime
        Scale Curve=300
       )
      )
     )
    )) {isSetParameters=true, module=SetVariables_#5}
   &FN_UpdateAnimTrail(SkeletalMesh=$User.SkeletalMesh)
   /SetVariables(Particles.RibbonU0Override=$Particles.NormalizedAge) {isSetParameters=true, module=SetVariables_#3}
   &ScaleColor(
    Scale Alpha=&FloatFromCurve(
     FloatCurve=DataInterface() {propertyValues {Curve {keys [
         <at 0.02 0 RCIM_Cubic>
         {arriveTangent=-1.3016, interpMode=RCIM_Cubic, leaveTangent=-1.3016, tangentMode=RCTM_User, time=0.1, value=1}
         <at 1 0 RCIM_Cubic>
        ]}}}
     Scale Curve=&FloatFromCurve(
      FloatCurve=DataInterface() {propertyValues {Curve {keys [{arriveTangent=0.1851, interpMode=RCIM_Cubic, leaveTangent=0.1851, tangentMode=RCTM_User, time=0.5, value=0}, <at 1 1 RCIM_Cubic>]}}}
      CurveIndex=$System.Timeline.NormalizedTime
     )
    )
    Scale RGB=&MakeCustomVectorFromBool(
     Bool=$System.IsDefaultVariant
     True Vector=&MakeVectorFromLinearColorRGB(Color=&ColorFromCurve(ColorCurve=DataInterface() {propertyValues {
         AlphaCurve {keys []}
         BlueCurve {keys [<at 0.0886 1.4 RCIM_Cubic>,<at 0.7726 0.2 RCIM_Cubic>]}
         GreenCurve {keys [<at 0.0886 1.4 RCIM_Cubic>,<at 0.7726 0.28 RCIM_Cubic>]}
         RedCurve {keys [<at 0.0886 2 RCIM_Cubic>,<at 0.7726 1 RCIM_Cubic>]}
        }}))
     False Vector=&MakeVectorFromLinearColorRGB(Color=&ColorFromCurve(ColorCurve=DataInterface() {propertyValues {
         AlphaCurve {keys []}
         BlueCurve {keys [<at 0.0886 2 RCIM_Cubic>,<at 0.7726 1 RCIM_Cubic>]}
         GreenCurve {keys [<at 0.0886 1.2833 RCIM_Cubic>,<at 0.7726 0.25 RCIM_Cubic>]}
         RedCurve {keys [<at 0.0886 1 RCIM_Cubic>,<at 0.7726 0.1 RCIM_Cubic>]}
        }}))
    )
    ScaleA=true
    ScaleRGB=true
   )
   &ScaleColor(
    Scale Alpha=&FloatFromCurve(FloatCurve=DataInterface() {propertyValues {Curve {keys [<at 3.5 0 RCIM_Cubic>,<at 4.5 1 RCIM_Cubic>]}}}, CurveIndex=$Emitter.Age)
    ScaleA=true
    ScaleRGB=false
   ) {module=ScaleColor001}
   &ScaleColor(
    Scale RGB=&VectorFromFloat(Value=&FloatFromCurve(
      FloatCurve=DataInterface() {propertyValues {Curve {keys [
          <at 0.1 1 RCIM_Cubic>
          {arriveTangent=0.1913, interpMode=RCIM_Cubic, leaveTangent=0.1913, tangentMode=RCTM_User, time=0.3, value=8}
          {arriveTangent=-8.9551, interpMode=RCIM_Cubic, leaveTangent=-8.9551, tangentMode=RCTM_User, time=0.4, value=4}
          <at 1 1 RCIM_Cubic>
         ]}}}
      CurveIndex=$System.Timeline002.NormalizedTime
     ))
    ScaleA=false
    ScaleRGB=true
   ) {module=ScaleColor002}
   &ScaleColor(
    Scale Alpha=&MaskFloatBySpawnGroup(
     Masked Float=&Add_Float(A=1, B=&Multiply_Float(A=$System.InterpolateOverTime.InterpolateOverTime_Float.Moving Average, B=1))
     Passthrough Float=&Subtract_Float(A=1, B=&Multiply_Float(A=$System.InterpolateOverTime.InterpolateOverTime_Float.Moving Average, B=1))
    )
    ScaleA=true
    ScaleRGB=false
   ) {module=ScaleColor003}
  ]
  rendererClasses [&NiagaraRibbonRendererProperties]
  renderers [{rendererClass=&NiagaraRibbonRendererProperties}]
 }
 SmokeBurstDeploy {
  emitterUpdateScript [
   &EmitterState()
   &SpawnRate(
    SpawnRate=&FloatFromCurve(
     FloatCurve=DataInterface() {propertyValues {Curve {keys [
         {arriveTangent=0.2667, interpMode=RCIM_Cubic, time=3.2, value=0}
         {arriveTangent=-0.0798, interpMode=RCIM_Cubic, leaveTangent=-0.0798, tangentMode=RCTM_User, time=3.4, value=1}
         <at 4.2 0 RCIM_Cubic>
        ]}}}
     CurveIndex=$Emitter.Age
     Scale Curve=250
    )
    Use Spawn Probability=false
   )
   &SpawnBurst_Instantaneous(
    Spawn Count=&RandomRangeInt(Minimum=2, Maximum=3)
    Use Spawn Probability=false
    UseLoopCountLimit=false
   ) {enabled=false}
  ]
  particleSpawnScript [
   &InitializeParticle(
    Color Maximum=<rgba 0.076 0.1 0.1 0.9>
    Color Minimum=<rgba 0.048 0.055 0.06 0.2>
    Color Mode=::NewEnumerator2
    Lifetime Max=0.8
    Lifetime Min=0.2
    Lifetime Mode=::NewEnumerator1
    Sprite Rotation Mode=::NewEnumerator2
    Sprite Size Mode=::NewEnumerator1
    Uniform Sprite Size Max=180
    Uniform Sprite Size Min=50
    UsePositionOffset=false
    Write Material Random=true
   )
   &CameraOffset(Camera Offset Amount=&RandomRangeFloat(Minimum=40, Maximum=50, Evaluation Type=::NewEnumerator1)) {enabled=false}
   &PartitionParticles()
   &SkeletalMeshLocation(
    Bone / Socket Index=$Particles.PartitionParticles.Partition
    Bone Sampling Mode=::NewEnumerator4
    MaskSpawnGroup=false
    Position Sampling=::NewEnumerator2
    Skeletal Mesh=$User.SkeletalMesh
   ) {enabled=false}
   &ShapeLocation(
    MaskSpawnGroup=false
    Non Uniform Scale=<1 0 1>
    Override Seed=false
    Shape Origin=$Particles.SkeletalMeshLocation.SampledPosition
    Sphere Radius=20
    UseOwnerScale=true
   ) {enabled=false, module=ShapeLocation001}
   &AddVelocity(Velocity=&RotateVectorByQuaternion(Vector To Rotate=<0 0 1000>, Quaternion=$Particles.SkeletalMeshLocation.SampledRotation)) {enabled=false, module=AddVelocity001}
  ]
  particleUpdateScript [
   &ParticleState()
   &SkeletalMeshLocation(
    Bone / Socket Index=$Particles.PartitionParticles.Partition
    Bone Sampling Mode=::NewEnumerator4
    MaskSpawnGroup=false
    Orientation Sampling=::NewEnumerator1
    Skeletal Mesh=$User.SkeletalMesh
   ) {module=SkeletalMeshLocation001}
   /SetVariables(Particles.Position=&AddVectorToPosition(
     Position=$Particles.Position
     Vector=&Multiply_VectorByFloat(
      Vector=&RotateVectorByQuaternion(
       Vector To Rotate=&MakeVector(
        X=0
        Y=&Multiply_Float(
         A=&Lerp_Float(A=0.6, B=1.4, Alpha=$System.Timeline.NormalizedTime)
         B=&Add_Float(
          A=&MaskFloatBySpawnGroup(Masked Float=1, Passthrough Float=-1, Spawn Group Mask=$Particles.PartitionParticles.Partition)
          B=&Multiply_Float(A=$System.InterpolateOverTime.InterpolateOverTime_Float.Moving Average, B=0.5)
         )
        )
        Z=0
       )
       Quaternion=$Particles.SkeletalMeshLocation001.SampledRotation
      )
      Float=&FloatFromCurve(FloatCurve=DataInterface() {propertyValues {Curve {keys [<at 0 0>,<at 1 1>]}}}, Scale Curve=300)
     )
    )) {isSetParameters=true}
   &ScaleSpriteSize(Uniform Curve Sprite Scale=DataInterface() {propertyValues {Curve {keys [<at 0 0.5 RCIM_Cubic>,<at 1 1 RCIM_Cubic>]}}})
   &ScaleColor(
    Scale Alpha=&FloatFromCurve(
     FloatCurve=DataInterface() {propertyValues {Curve {keys [
         <at 0 0 RCIM_Cubic>
         <at 0.15 0.5 RCIM_Cubic>
         {arriveTangent=-0.0674, interpMode=RCIM_Cubic, leaveTangent=-0.0674, tangentMode=RCTM_User, time=1, value=0}
        ]}}}
     Scale Curve=&FN_Select_Float_from_Time_Of_Day(Morning=0.3, Day=0.3, Evening=0.3, Night=0.4)
    )
    Scale RGB=&MakeCustomVectorFromBool(
     Bool=$System.IsDefaultVariant
     True Vector=&MakeVectorFromLinearColorRGB(Color=&ColorFromCurve(ColorCurve=DataInterface() {propertyValues {
         AlphaCurve {keys []}
         BlueCurve {keys [<at 0.0733 0.1 RCIM_Cubic>,<at 0.4235 0.5 RCIM_Cubic>]}
         GreenCurve {keys [<at 0.0733 0.595 RCIM_Cubic>,<at 0.4235 0.5 RCIM_Cubic>]}
         RedCurve {keys [<at 0.0733 10 RCIM_Cubic>,<at 0.4235 0.5 RCIM_Cubic>]}
        }}))
     False Vector=&MakeVectorFromLinearColorRGB(Color=&ColorFromCurve(ColorCurve=DataInterface() {propertyValues {
         AlphaCurve {keys []}
         BlueCurve {keys [<at 0.0733 10 RCIM_Cubic>,<at 0.4235 0.5 RCIM_Cubic>]}
         GreenCurve {keys [<at 0.0733 3.4 RCIM_Cubic>,<at 0.4235 0.5 RCIM_Cubic>]}
         RedCurve {keys [<at 0.0733 0.1 RCIM_Cubic>,<at 0.4235 0.5 RCIM_Cubic>]}
        }}))
    )
    ScaleA=true
    ScaleRGB=true
   )
   &ScaleVelocity(Velocity Scale=&MakeVector(
     X=1
     Y=&FloatFromCurve(
      FloatCurve=DataInterface() {propertyValues {Curve {keys [{arriveTangent=2.0742, interpMode=RCIM_Cubic, leaveTangent=2.0742, tangentMode=RCTM_User, time=0, value=0.1}, <at 1 1 RCIM_Cubic>]}}}
      Scale Curve=0
     )
     Z=1
    )) {enabled=false}
   &DynamicMaterialParameters(
    Index 0 Param 1=&Add_Float(
     A=&RandomRangeFloat(Minimum=0, Maximum=1)
     B=&FloatFromCurve(
      FloatCurve=DataInterface() {propertyValues {Curve {keys [
          <at 0 0 RCIM_Cubic>
          {arriveTangent=1.5947, interpMode=RCIM_Cubic, leaveTangent=1.5947, tangentMode=RCTM_User, time=0.2557, value=0.5257}
          <at 1 1 RCIM_Cubic>
         ]}}}
      CurveIndex=$Particles.Age
      Scale Curve=0.1
     )
    )
    Module.Param0WriteEnabled=true
    Module.Param1WriteEnabled=true
    Module.Param2WriteEnabled=true
    Module.Param3WriteEnabled=true
   ) {module=DynamicMaterialParameters001}
   &GravityForce(Gravity=<0 0 -100>) {enabled=false}
   &SolveForcesAndVelocity(Clamp Velocity=false, Limit Acceleration=false)
  ]
  rendererClasses [&NiagaraSpriteRendererProperties]
  renderers [
   {rendererClass=&NiagaraSpriteRendererProperties}
   {rendererClass=&NiagaraSpriteRendererProperties, rendererIndex=1}
  ]
 }
 SpeedLine {
  emitterUpdateScript [
   &EmitterState(Inactive Response=::NewEnumerator1, Life Cycle Mode=::NewEnumerator1, Loop Delay=1.5, Loop Duration=3, UseLoopDelay=false)
   &SpawnRate(
    SpawnRate=&FN_FrontEndFloatModifier(
     Gameplay=&Multiply_Float(A=150, B=&Lerp_Float(A=0, B=1, Alpha=$System.Timeline.NormalizedTime))
     Front End=0
    )
    Use Spawn Probability=false
   )
  ]
  particleSpawnScript [
   &InitializeParticle(
    Lifetime Max=0.2
    Lifetime Min=0.2
    Lifetime Mode=::NewEnumerator1
    Sprite Size Max=<2 7>
    Sprite Size Min=<2 2>
    Sprite Size Mode=::NewEnumerator2
    UsePositionOffset=false
    Write Material Random=true
   )
   &SkeletalMeshLocation(
    Bone Sampling Mode=::NewEnumerator4
    Interpolate Mesh Sampling=false
    MaskSpawnGroup=false
    Orientation Sampling=::NewEnumerator1
    Sampled Position Offset=<0 0 50>
    Skeletal Mesh=DataInterface() {propertyValues {
      DefaultMesh=null
      ExcludeBoneName=root
      OptionalComponentClass=null
      PreviewMesh=/BRCosmetics/Gadgets/Assets/VinderTech_GliderChute/Glider_Foe_Male/Meshes/Foe_Male_Glider
      SkinningMode=SkinOnTheFly
      SoftSourceActor=null
      SourceMode=Default
      UvSetIndex=0
      WholeMeshLOD=-1
      bExcludeBone=false
      bReadDeformedGeometry=true
      bRequireCurrentFrameData=true
      ComponentTags []
      FilteredBones [Glider_Base]
      FilteredSockets []
      MeshUserParameter {parameter {name=None, typeDefHandle {registeredTypeIndex=1}, varData []}}
      SamplingRegions []
     }}
   ) {module=SkeletalMeshLocation001}
   &TorusLocation(
    Align Torus With Axis=true
    Handle Radius=0
    Large Radius:NiagaraFloat=&RandomRangeFloat(Minimum=&FN_FrontEndFloatModifier(Gameplay=200, Front End=0), Maximum=300)
    MaskSpawnGroup=false
    Module.Override Randomness=false
    Offset=&FN_FrontEndVector3Modifier(Gameplay=<0 100 0>, Front End=<0 -100 0>)
    Torus Axis=<0 1 0>
   )
   &AddVelocity(Velocity=&RandomRangeVector(Minimum=<0 -1000 0>, Maximum=<0 -2000 0>))
   &AddVelocityFromPoint(Module.Use Falloff Distance=false, Origin Offset=<0 200 0>, Velocity Strength=700)
   &CameraOffset(Camera Offset Amount=-10)
  ]
  particleUpdateScript [
   &SolveForcesAndVelocity(Clamp Velocity=false, Limit Acceleration=false)
   &ParticleState()
   &ScaleSpriteSizeBySpeed(Max Scale Factor=<1 50>, Min Scale Factor=<0.01 0.05>, Velocity Threshold=1)
   &ScaleColor(
    Scale Alpha=&Multiply_Float(
     A=&FloatFromCurve(FloatCurve=DataInterface() {propertyValues {Curve {keys [
          {interpMode=RCIM_Cubic, tangentMode=RCTM_User, time=0, value=0}
          {interpMode=RCIM_Cubic, tangentMode=RCTM_User, time=0.5, value=1}
          {tangentMode=RCTM_User, time=1, value=0}
         ]}}})
     B=&RandomRangeFloat(Minimum=0.1, Maximum=0.3)
    )
    ScaleA=true
    ScaleRGB=true
   )
   &ScaleColor(
    Scale Alpha=&FloatFromCurve(
     FloatCurve=DataInterface() {propertyValues {Curve {keys [<at 0.6 0 RCIM_Cubic>,<at 1 1 RCIM_Cubic>]}}}
     CurveIndex=$System.Timeline.NormalizedTime
    )
    ScaleA=true
    ScaleRGB=false
   ) {module=ScaleColor001}
  ]
  rendererClasses [&NiagaraSpriteRendererProperties]
  renderers [{rendererClass=&NiagaraSpriteRendererProperties}]
 }
 TrailBurstDeploy {
  emitterUpdateScript [
   &EmitterState(Life Cycle Mode=::NewEnumerator1, Loop Delay=3, Loop Duration=1, UseLoopDelay=false)
   &FN_SpawnAnimTrail(
    Enable:NiagaraBool=&SetBoolByFloatComparison(A=$System.Timeline.NormalizedTime, B=0)
    Number of Anim Trails=2
    Spawn Interval=0
   ) {enabled=false}
   &FN_SpawnAnimTrail(
    Enable:NiagaraBool=&Boolean_AndOperation(
     Bool1=&SetBoolByFloatComparison(A=$Emitter.Age, B=2.5)
     Bool2=&SetBoolByFloatComparison(A=$Emitter.Age, B=6, Comparison Type=::NewEnumerator6)
    )
    Number of Anim Trails=2
   ) {module=FN_SpawnAnimTrail001}
  ]
  particleSpawnScript [
   &FN_InitializeAnimTrailBoneIndecies()
   &FN_InitializeAnimTrailBoneIndecies(AnimTrail Selection=1, First Bone or Socket Index=2, Second Bone or Socket Index=3) {module=FN_InitializeAnimTrailBoneIndecies001}
   &FN_InitializeAnimTrail(Color=&FN_Select_Color_from_Time_Of_Day(Morning=<rgba 1 1 1 0.8>, Day=<rgba 1 1 1 1>, Evening=<rgba 1 1 1 0.8>, Night=<rgba 1 1 1 0.4>))
   &DynamicMaterialParameters(
    Index 0 Param 1=$User.Variant
    Module.Param0WriteEnabled=true
    Module.Param1WriteEnabled=false
    Module.Param2WriteEnabled=false
    Module.Param3WriteEnabled=false
   )
  ]
  particleUpdateScript [
   &SkeletalMeshLocation(
    Bone / Socket Index=$Particles.SpawnGroup
    Bone Sampling Mode=::NewEnumerator4
    MaskSpawnGroup=false
    Orientation Sampling=::NewEnumerator1
    Skeletal Mesh=$User.SkeletalMesh
   )
   /SetVariables(Particles.Position=&AddVectorToPosition(
     Position=$Particles.Position
     Vector=&Multiply_VectorByFloat(
      Vector=&RotateVectorByQuaternion(
       Vector To Rotate=&MakeVector(X=0, Y=&MaskFloatBySpawnGroup(Masked Float=1, Passthrough Float=-1), Z=0)
       Quaternion=$Particles.SkeletalMeshLocation.SampledRotation
      )
      Float=&FloatFromCurve(
       FloatCurve=DataInterface() {propertyValues {Curve {keys [<at 0 0>,<at 1 1>]}}}
       Scale Curve=&Multiply_Float(
        A=300
        B=&Lerp_Float(
         A=0.6
         B=1.2
         Alpha=&NormalizeFloat(
          Value=&MakeFloatFromVector(
           Vector=&RotateVectorByQuaternion(
            Vector To Rotate=&MaskVectorBySpawnGroup(Masked Vector=<0 0 -1>, Passthrough Vector=<0 0 1>)
            Quaternion=$Particles.SkeletalMeshLocation.SampledRotation
           )
           Channel=::NewEnumerator2
          )
          Min=0
          Max=1
         )
        )
       )
      )
     )
    )) {enabled=false, isSetParameters=true}
   /SetVariables(Particles.Position=&AddVectorToPosition(
     Position=$Particles.Position
     Vector=&Multiply_VectorByFloat(
      Vector=&RotateVectorByQuaternion(
       Vector To Rotate=&MakeVector(X=0, Y=&MaskFloatBySpawnGroup(Masked Float=1, Passthrough Float=-1), Z=0)
       Quaternion=$Particles.SkeletalMeshLocation.SampledRotation
      )
      Float=&FloatFromCurve(
       FloatCurve=DataInterface() {propertyValues {Curve {keys [<at 0 0>,<at 1 1>]}}}
       Scale Curve=&FloatFromCurve(
        FloatCurve=DataInterface() {propertyValues {Curve {keys [
            <at 3.1 0 RCIM_Cubic>
            {arriveTangent=-0.0638, interpMode=RCIM_Cubic, leaveTangent=-0.0638, tangentMode=RCTM_User, time=3.4, value=1}
            {arriveTangent=-0.0372, interpMode=RCIM_Cubic, leaveTangent=-0.0372, tangentMode=RCTM_User, time=3.8, value=1}
            <at 4.8 0 RCIM_Cubic>
           ]}}}
        CurveIndex=$Emitter.Age
        Scale Curve=&FN_FrontEndFloatModifier(Gameplay=300, Front End=200)
       )
      )
     )
    )) {isSetParameters=true, module=SetVariables_#5}
   &FN_UpdateAnimTrail(SkeletalMesh=$User.SkeletalMesh)
   /SetVariables(Particles.RibbonU0Override=$Particles.NormalizedAge) {isSetParameters=true, module=SetVariables_#3}
   &ScaleColor(
    Scale Alpha=&FloatFromCurve(
     FloatCurve=DataInterface() {propertyValues {Curve {keys [
         <at 0.01 0 RCIM_Cubic>
         {arriveTangent=-1.3016, interpMode=RCIM_Cubic, leaveTangent=-1.3016, tangentMode=RCTM_User, time=0.1, value=1}
         <at 1 0 RCIM_Cubic>
        ]}}}
     Scale Curve=&FloatFromCurve(
      FloatCurve=DataInterface() {propertyValues {Curve {keys [{arriveTangent=0.1851, interpMode=RCIM_Cubic, leaveTangent=0.1851, tangentMode=RCTM_User, time=0.5, value=1}, <at 1 0 RCIM_Cubic>]}}}
      CurveIndex=$System.Timeline.NormalizedTime
     )
    )
    Scale RGB=&MakeCustomVectorFromBool(
     Bool=$System.IsDefaultVariant
     True Vector=&MakeVectorFromLinearColorRGB(Color=&ColorFromCurve(ColorCurve=DataInterface() {propertyValues {
         AlphaCurve {keys []}
         BlueCurve {keys [<at 0.1282 0.5 RCIM_Cubic>,<at 0.7134 1 RCIM_Cubic>]}
         GreenCurve {keys [<at 0.1282 0.5 RCIM_Cubic>,<at 0.7134 1 RCIM_Cubic>]}
         RedCurve {keys [<at 0.1282 1 RCIM_Cubic>,<at 0.7134 1 RCIM_Cubic>]}
        }}))
     False Vector=&MakeVectorFromLinearColorRGB(Color=&ColorFromCurve(ColorCurve=DataInterface() {propertyValues {
         AlphaCurve {keys []}
         BlueCurve {keys [<at 0.1282 1 RCIM_Cubic>,<at 0.7134 1 RCIM_Cubic>]}
         GreenCurve {keys [<at 0.1282 0.5833 RCIM_Cubic>,<at 0.7134 1 RCIM_Cubic>]}
         RedCurve {keys [<at 0.1282 0.5 RCIM_Cubic>,<at 0.7134 1 RCIM_Cubic>]}
        }}))
    )
    ScaleA=true
    ScaleRGB=true
   )
   &ScaleColor(
    Scale Alpha=&FloatFromCurve(
     FloatCurve=DataInterface() {propertyValues {Curve {keys [<at 4.8 1 RCIM_Cubic>,<at 5.5 0 RCIM_Cubic>]}}}
     CurveIndex=$Emitter.Age
     Scale Curve=&FN_FrontEndFloatModifier(Gameplay=1.5, Front End=0.8)
    )
    ScaleA=true
    ScaleRGB=false
   ) {module=ScaleColor001}
  ]
  rendererClasses [&NiagaraRibbonRendererProperties]
  renderers [{rendererClass=&NiagaraRibbonRendererProperties}]
 }
}
summary {
 systemName=NS_Glider_RippedHarvester_Trails
 emitters [
  {emitterName=HeatDisks, rendererClasses [&NiagaraSpriteRendererProperties]}
  {emitterName=Embers, rendererClasses [&NiagaraSpriteRendererProperties]}
  {emitterName=SmokeBurstDeploy, rendererClasses [&NiagaraSpriteRendererProperties]}
  {emitterName=Main_Trail, rendererClasses [&NiagaraRibbonRendererProperties]}
  {emitterName=EngineLight, rendererClasses [&NiagaraSpriteRendererProperties, &NiagaraLightRendererProperties]}
  {emitterName=HeatCones, rendererClasses [&NiagaraMeshRendererProperties]}
  {emitterName=SpeedLine, rendererClasses [&NiagaraSpriteRendererProperties]}
  {emitterName=BaseTrail, rendererClasses [&NiagaraRibbonRendererProperties]}
  {emitterName=TrailBurstDeploy, rendererClasses [&NiagaraRibbonRendererProperties]}
  {emitterName=FlashBurstDeploy, rendererClasses [&NiagaraSpriteRendererProperties]}
  {emitterName=FlashBurstForward, rendererClasses [&NiagaraSpriteRendererProperties]}
 ]
 userVariables [
  User.bIsFrontEnd {type=NiagaraBool}
  User.bIsFrontEndPreview {type=NiagaraBool}
  User.bIsFullyDeployed {type=NiagaraBool}
  User.ForwardDot {type=NiagaraFloat, defaultValue=0, description=""}
  User.PlayerSpeed {type=NiagaraFloat, defaultValue=0, description=""}
  User.RibbonIDOffset {type=NiagaraInt32}
  User.RightDot {type=NiagaraFloat, defaultValue=0, description=""}
  User.SkeletalMesh {
   type=NiagaraDataInterfaceSkeletalMesh
   defaultValue {
    struct=NiagaraExt_VariableValue_DataInterface
    value {dataInterface=&self:NiagaraDataInterfaceSkeletalMesh_0}
   }
  }
  User.Variant {type=NiagaraFloat, defaultValue=0, description=""}
 ]
}
